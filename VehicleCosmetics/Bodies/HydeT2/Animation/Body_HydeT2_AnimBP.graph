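# Animation Blueprint that procedurally poses the HydeT2 vehicle body and layers boost and engine rumble animation on it.
# Chained Modify Bone nodes drive axles, wheels and steering from variables not declared here, starting from the ref pose.
# bHasTireMod pushes wheels outward and widens wheel wells with interpolated blending, before a bouncy chassis controller.
# Two additive state machines follow: four-stage boost and driver-triggered rumble.
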AnimBlueprint /VehicleCosmetics/Bodies/HydeT2/Animation/Body_HydeT2_AnimBP
target_skeleton=/VehicleCosmetics/Bodies/HydeT2/Meshes/SK_HydeT2_LOD0_Skeleton
graphs [
 AnimGraph {nodes [
   n1 = "Transform (Modify) Bone - Bone: body"(
    ComponentPose=Mesh Space Ref Pose!() -> (ComponentPose:Component Space Pose Link Structure) {is=title AnimGraphNode_MeshRefPose}
    Alpha=1
    Translation="0,0,0"
    Rotation="-0.5,0,0"
    Scale="1,1,1"
   ) {
    rotationMode=BMM_Replace
    rotationSpace=BCS_BoneSpace
    translationMode=BMM_Replace
    translationSpace=BCS_BoneSpace
    alphaBoolBlend {blendOption=Linear}
    boneToModify {boneName=body}
    rotation {pitch=-0.5}
   }
   n2 = "Transform (Modify) Bone - Bone: front_differential"(
    ComponentPose=@n1
    Alpha=1
    Translation=Get AxlePositionF() -> (AxlePositionF:Vector)
    Rotation=Get AxleRotationF() -> (AxleRotationF:Rotator)
    Scale="1,1,1"
   ) {
    rotationMode=BMM_Replace
    rotationSpace=BCS_BoneSpace
    translationMode=BMM_Replace
    translationSpace=BCS_BoneSpace
    alphaBoolBlend {blendOption=Linear}
    boneToModify {boneName=front_differential}
   }
   n3 = "Transform (Modify) Bone - Bone: rear_differential"(
    ComponentPose=@n2
    Alpha=1
    Translation=Get AxlePositionB() -> (AxlePositionB:Vector)
    Rotation=Get AxleRotationB() -> (AxleRotationB:Rotator)
    Scale="1,1,1"
   ) {
    rotationMode=BMM_Replace
    rotationSpace=BCS_BoneSpace
    translationMode=BMM_Replace
    translationSpace=BCS_BoneSpace
    alphaBoolBlend {blendOption=Linear}
    boneToModify {boneName=rear_differential}
   }
   n4 = "Transform (Modify) Bone - Bone: axle_pivot_fr_l"(ComponentPose=@n3, Alpha=1, Translation=Get WheelPositionFL() -> (WheelPositionFL:Vector), Rotation="0,0,0", Scale="1,1,1") {
    rotationMode=BMM_Replace
    rotationSpace=BCS_BoneSpace
    translationMode=BMM_Replace
    translationSpace=BCS_BoneSpace
    alphaBoolBlend {blendOption=Linear}
    boneToModify {boneName=axle_pivot_fr_l}
   }
   n5 = "Transform (Modify) Bone - Bone: axle_pivot_fr_r"(ComponentPose=@n4, Alpha=1, Translation=Get WheelPositionFR() -> (WheelPositionFR:Vector), Rotation="0,0,0", Scale="1,1,1") {
    rotationMode=BMM_Replace
    rotationSpace=BCS_BoneSpace
    translationMode=BMM_Replace
    translationSpace=BCS_BoneSpace
    alphaBoolBlend {blendOption=Linear}
    boneToModify {boneName=axle_pivot_fr_r}
   }
   n6 = "Transform (Modify) Bone - Bone: axle_pivot_bk_r"(ComponentPose=@n5, Alpha=1, Translation=Get WheelPositionBR() -> (WheelPositionBR:Vector), Rotation="0,0,0", Scale="1,1,1") {
    rotationMode=BMM_Replace
    rotationSpace=BCS_BoneSpace
    translationMode=BMM_Replace
    translationSpace=BCS_BoneSpace
    alphaBoolBlend {blendOption=Linear}
    boneToModify {boneName=axle_pivot_bk_r}
   }
   n7 = "Transform (Modify) Bone - Bone: axle_pivot_bk_l"(ComponentPose=@n6, Alpha=1, Translation=Get WheelPositionBL() -> (WheelPositionBL:Vector), Rotation="0,0,0", Scale="1,1,1") {
    rotationMode=BMM_Replace
    rotationSpace=BCS_BoneSpace
    translationMode=BMM_Replace
    translationSpace=BCS_BoneSpace
    alphaBoolBlend {blendOption=Linear}
    boneToModify {boneName=axle_pivot_bk_l}
   }
   n8 = "Transform (Modify) Bone - Bone: wheel_steering_fr_r"(ComponentPose=@n7, Alpha=1, Translation="0,0,0", Rotation=Get WheelRotationFR() -> (WheelRotationFR:Rotator), Scale="1,1,1") {
    rotationMode=BMM_Replace
    rotationSpace=BCS_BoneSpace
    translationSpace=BCS_BoneSpace
    alphaBoolBlend {blendOption=Linear}
    boneToModify {boneName=wheel_steering_fr_r}
   }
   n9 = "Transform (Modify) Bone - Bone: wheel_steering_fr_l"(ComponentPose=@n8, Alpha=1, Translation="0,0,0", Rotation=Get WheelRotationFL() -> (WheelRotationFL:Rotator), Scale="1,1,1") {
    rotationMode=BMM_Replace
    rotationSpace=BCS_BoneSpace
    translationSpace=BCS_BoneSpace
    alphaBoolBlend {blendOption=Linear}
    boneToModify {boneName=wheel_steering_fr_l}
   }
   n10 = "Transform (Modify) Bone - Bone: wheel_disc_fr_r"(ComponentPose=@n9, Alpha=1, Translation="0,0,0", Rotation=Get WheelSpinFR() -> (WheelSpinFR:Rotator), Scale="1,1,1") {rotationMode=BMM_Replace, rotationSpace=BCS_BoneSpace, alphaBoolBlend {blendOption=Linear}, boneToModify {boneName=wheel_disc_fr_r}}
   n11 = "Transform (Modify) Bone - Bone: wheel_disc_fr_l"(ComponentPose=@n10, Alpha=1, Translation="0,0,0", Rotation=Get WheelSpinFL() -> (WheelSpinFL:Rotator), Scale="1,1,1") {rotationMode=BMM_Replace, rotationSpace=BCS_BoneSpace, alphaBoolBlend {blendOption=Linear}, boneToModify {boneName=wheel_disc_fr_l}}
   n12 = "Transform (Modify) Bone - Bone: wheel_disc_bk_r"(ComponentPose=@n11, Alpha=1, Translation="0,0,0", Rotation=Get WheelSpinBR() -> (WheelSpinBR:Rotator), Scale="1,1,1") {rotationMode=BMM_Replace, rotationSpace=BCS_BoneSpace, alphaBoolBlend {blendOption=Linear}, boneToModify {boneName=wheel_disc_bk_r}}
   n13 = "Transform (Modify) Bone - Bone: wheel_disc_bk_l"(ComponentPose=@n12, Alpha=1, Translation="0,0,0", Rotation=Get WheelSpinBL() -> (WheelSpinBL:Rotator), Scale="1,1,1") {rotationMode=BMM_Replace, rotationSpace=BCS_BoneSpace, alphaBoolBlend {blendOption=Linear}, boneToModify {boneName=wheel_disc_bk_l}}
   n14 = "Transform (Modify) Bone - Bone: steering_wheel"(ComponentPose=@n13, Alpha=1, Translation="0,0,0", Rotation=Get SteeringRotation() -> (SteeringRotation:Rotator), Scale="1,1,1") {rotationMode=BMM_Replace, rotationSpace=BCS_BoneSpace, alphaBoolBlend {blendOption=Linear}, boneToModify {boneName=steering_wheel}}
   n15 = Get bHasTireMod() -> (bHasTireMod:bool)
   n16 = "Transform (Modify) Bone - Bone: axle_pivot_bk_l"!(ComponentPose=@n14, bAlphaBoolEnabled:bool=@n15, Translation="0,-24,0", Rotation="0,0,0", Scale="1,1,1") -> (Pose) {
    alphaInputType=Bool
    rotationSpace=BCS_BoneSpace
    scaleSpace=BCS_BoneSpace
    translationMode=BMM_Additive
    translationSpace=BCS_BoneSpace
    alphaBoolBlend {blendOption=Linear}
    boneToModify {boneName=axle_pivot_bk_l}
    translation {y=-24}
   }
   n17 = "Transform (Modify) Bone - Bone: axle_pivot_bk_r"!(ComponentPose=@n16, bAlphaBoolEnabled:bool=@n15, Translation="0,24,0", Rotation="0,0,0", Scale="1,1,1") -> (Pose) {
    alphaInputType=Bool
    rotationSpace=BCS_BoneSpace
    scaleSpace=BCS_BoneSpace
    translationMode=BMM_Additive
    translationSpace=BCS_BoneSpace
    alphaBoolBlend {blendOption=Linear}
    boneToModify {boneName=axle_pivot_bk_r}
    translation {y=24}
   }
   n18 = "Transform (Modify) Bone - Bone: axle_pivot_fr_l"!(ComponentPose=@n17, bAlphaBoolEnabled:bool=@n15, Translation="0,-24,0", Rotation="0,0,0", Scale="1,1,1") -> (Pose) {
    alphaInputType=Bool
    rotationSpace=BCS_BoneSpace
    translationMode=BMM_Additive
    translationSpace=BCS_BoneSpace
    alphaBoolBlend {blendOption=Linear}
    boneToModify {boneName=axle_pivot_fr_l}
    translation {y=-24}
   }
   n19 = "Transform (Modify) Bone - Bone: axle_pivot_fr_r"!(ComponentPose=@n18, bAlphaBoolEnabled:bool=@n15, Translation="0,24,0", Rotation="0,0,0", Scale="1,1,1") -> (Pose) {
    alphaInputType=Bool
    rotationSpace=BCS_BoneSpace
    translationMode=BMM_Additive
    translationSpace=BCS_BoneSpace
    alphaBoolBlend {blendOption=Linear}
    boneToModify {boneName=axle_pivot_fr_r}
    translation {y=24}
   }
   n20 = "Transform (Modify) Bone - Bone: wheel_well_spreader_bk_l"!(ComponentPose=@n19, bAlphaBoolEnabled:bool=@n15, Translation="3,0,0", Rotation="0,0,0", Scale="1.25,1,1") -> (Pose) {
    alphaInputType=Bool
    rotationSpace=BCS_BoneSpace
    scaleMode=BMM_Additive
    scaleSpace=BCS_BoneSpace
    translationMode=BMM_Additive
    translationSpace=BCS_BoneSpace
    alphaBoolBlend {blendOption=Linear}
    alphaScaleBiasClamp {bInterpResult=true, interpSpeedDecreasing=5, interpSpeedIncreasing=5}
    boneToModify {boneName=wheel_well_spreader_bk_l}
    scale {x=1.25}
    translation {x=3}
   }
   n21 = "Transform (Modify) Bone - Bone: wheel_well_spreader_bk_r"!(ComponentPose=@n20, bAlphaBoolEnabled:bool=@n15, Translation="3,0,0", Rotation="0,0,0", Scale="1.25,1,1") -> (Pose) {
    alphaInputType=Bool
    rotationSpace=BCS_BoneSpace
    scaleMode=BMM_Additive
    scaleSpace=BCS_BoneSpace
    translationMode=BMM_Additive
    translationSpace=BCS_BoneSpace
    alphaBoolBlend {blendOption=Linear}
    alphaScaleBiasClamp {bInterpResult=true, interpSpeedDecreasing=5, interpSpeedIncreasing=5}
    boneToModify {boneName=wheel_well_spreader_bk_r}
    scale {x=1.25}
    translation {x=3}
   }
   n22 = "Transform (Modify) Bone - Bone: wheel_well_spreader_fr_l"!(ComponentPose=@n21, bAlphaBoolEnabled:bool=@n15, Translation="0,0,0", Rotation="0,0,0", Scale="1.25,1,1") -> (Pose) {
    alphaInputType=Bool
    rotationSpace=BCS_BoneSpace
    scaleMode=BMM_Additive
    scaleSpace=BCS_BoneSpace
    translationMode=BMM_Additive
    translationSpace=BCS_BoneSpace
    alphaBoolBlend {blendOption=Linear}
    alphaScaleBiasClamp {bInterpResult=true, interpSpeedDecreasing=5, interpSpeedIncreasing=5}
    boneToModify {boneName=wheel_well_spreader_fr_l}
    scale {x=1.25}
   }
   n23 = "Transform (Modify) Bone - Bone: wheel_well_spreader_fr_r"!(ComponentPose=@n22, bAlphaBoolEnabled:bool=@n15, Translation="0,0,0", Rotation="0,0,0", Scale="1.25,1,1") -> (Pose) {
    alphaInputType=Bool
    rotationSpace=BCS_BoneSpace
    scaleMode=BMM_Additive
    scaleSpace=BCS_BoneSpace
    translationMode=BMM_Additive
    translationSpace=BCS_BoneSpace
    alphaBoolBlend {blendOption=Linear}
    alphaScaleBiasClamp {bInterpResult=true, interpSpeedDecreasing=5, interpSpeedIncreasing=5}
    boneToModify {boneName=wheel_well_spreader_fr_r}
    scale {x=1.25}
   }
   AnimGraphNode_BouncyChassisVehicleController_1 = Vehicle Bouncy Chassis Controller!(
    ComponentPose:Component Space Pose Link Structure=@n23
    Alpha:float=1
    BouncyChassis:Bouncy Chassis Vehicle State Structure=(PitchDegrees=0,RollDegrees=0,TranslationZ=0)
   ) -> (Pose:Component Space Pose Link Structure) {is=title AnimGraphNode_VehicleBouncyChassisController, alphaBoolBlend {blendOption=Linear}}
   n24 = Component To Local(ComponentPose=@AnimGraphNode_BouncyChassisVehicleController_1)
   AnimGraphNode_StateMachine_2 = HydeT2_Boost() {is=title AnimGraphNode_StateMachine}
   AnimGraphNode_StateMachine_0 = Engine_Rumble State Machine() {is=title AnimGraphNode_StateMachine}
   Output Pose(Result=Apply Additive(
     Base=Apply Additive(Base=@n24, Additive=@AnimGraphNode_StateMachine_2) {alphaBoolBlend {blendOption=Linear}}
     Additive=@AnimGraphNode_StateMachine_0
    ) {alphaBoolBlend {blendOption=Linear}})
  ]}
 Engine_Rumble State Machine {
  entry=Empty_Driver
  path=AnimGraph.AnimGraphNode_StateMachine_0.Engine_Rumble State Machine
  ids=id
  nodes [
   AnimStateNode_0 = Empty_Driver() {is=title AnimStateNode}
   AnimStateNode_1 = Driver_Seat() {is=title AnimStateNode}
   AnimStateTransitionNode_0 = AnimStateTransitionNode() {bAllowInertializationForSelfTransitions=true}
   AnimStateTransitionNode_1 = AnimStateTransitionNode() {bAllowInertializationForSelfTransitions=true}
  ]
  links [
   {
    dst=Empty_Driver
    rule=AnimGraph.AnimGraphNode_StateMachine_0.Engine_Rumble State Machine.AnimStateTransitionNode_1.Transition
    src=Driver_Seat
    via=AnimStateTransitionNode_1
   }
   {
    dst=Driver_Seat
    rule=AnimGraph.AnimGraphNode_StateMachine_0.Engine_Rumble State Machine.AnimStateTransitionNode_0.Transition
    src=Empty_Driver
    via=AnimStateTransitionNode_0
   }
  ]
 }
 Empty_Driver {
  path=AnimGraph.AnimGraphNode_StateMachine_0.Engine_Rumble State Machine.AnimStateNode_0.Empty_Driver
  nodes [Output Animation Pose(Result="Sequence Player (additive) 'Body_Hyde_T2_engine_rumble_NO_ANIM'"() {bLoopAnimation=false, sequence=&/Body_Hyde_T2_engine_rumble_NO_ANIM})]
 }
 Driver_Seat {
  path=AnimGraph.AnimGraphNode_StateMachine_0.Engine_Rumble State Machine.AnimStateNode_1.Driver_Seat
  nodes [
   n1 = "Sequence Player (additive) 'Body_Hyde_T2_engine_rumble_loop'"() {playRate=1.5, sequence=&/Body_Hyde_T2_engine_rumble_loop, playRateScaleBiasClampConstants {scale=2}}
   Output Animation Pose(Result=@n1)
  ]
 }
 Transition {
  path=AnimGraph.AnimGraphNode_StateMachine_0.Engine_Rumble State Machine.AnimStateTransitionNode_0.Transition
  nodes [Result(bCanEnterTransition=Get bHasDriver() -> (bHasDriver:bool))]
 }
 Transition {
  path=AnimGraph.AnimGraphNode_StateMachine_0.Engine_Rumble State Machine.AnimStateTransitionNode_1.Transition
  nodes [Result(bCanEnterTransition=NOT Boolean(A=Get bHasDriver()))]
 }
 HydeT2_Boost {
  entry=Idle
  path=AnimGraph.AnimGraphNode_StateMachine_2.HydeT2_Boost
  ids=id
  nodes [
   AnimStateNode_0 = Idle()
   AnimStateNode_1 = Boost_Start() {is=title AnimStateNode}
   AnimStateNode_2 = Boost_Loop() {is=title AnimStateNode}
   AnimStateNode_3 = Boost_End() {is=title AnimStateNode}
   AnimStateTransitionNode_0 = AnimStateTransitionNode() {bAllowInertializationForSelfTransitions=true}
   AnimStateTransitionNode_1 = AnimStateTransitionNode() {bAllowInertializationForSelfTransitions=true}
   AnimStateTransitionNode_2 = AnimStateTransitionNode() {bAllowInertializationForSelfTransitions=true}
   AnimStateTransitionNode_3 = AnimStateTransitionNode() {bAllowInertializationForSelfTransitions=true}
  ]
  links [
   {
    dst=Idle
    rule=AnimGraph.AnimGraphNode_StateMachine_2.HydeT2_Boost.AnimStateTransitionNode_3.Transition
    src=Boost_End
    via=AnimStateTransitionNode_3
   }
   {
    dst=Boost_End
    rule=AnimGraph.AnimGraphNode_StateMachine_2.HydeT2_Boost.AnimStateTransitionNode_2.Transition
    src=Boost_Loop
    via=AnimStateTransitionNode_2
   }
   {
    dst=Boost_Loop
    rule=AnimGraph.AnimGraphNode_StateMachine_2.HydeT2_Boost.AnimStateTransitionNode_1.Transition
    src=Boost_Start
    via=AnimStateTransitionNode_1
   }
   {
    dst=Boost_Start
    rule=AnimGraph.AnimGraphNode_StateMachine_2.HydeT2_Boost.AnimStateTransitionNode_0.Transition
    src=Idle
    via=AnimStateTransitionNode_0
   }
  ]
 }
 Idle {
  path=AnimGraph.AnimGraphNode_StateMachine_2.HydeT2_Boost.AnimStateNode_0.Idle
  nodes [Output Animation Pose(Result="Sequence Player (additive) 'Body_Hyde_T2_boost_NO_ANIM'"() {bLoopAnimation=false, sequence=&/Body_Hyde_T2_boost_NO_ANIM})]
 }
 Boost_Start {
  path=AnimGraph.AnimGraphNode_StateMachine_2.HydeT2_Boost.AnimStateNode_1.Boost_Start
  nodes [Output Animation Pose(Result="Sequence Player (additive) 'Body_Hyde_T2_boost_start'"() {bLoopAnimation=false, sequence=&/Body_Hyde_T2_boost_start})]
 }
 Boost_Loop {
  path=AnimGraph.AnimGraphNode_StateMachine_2.HydeT2_Boost.AnimStateNode_2.Boost_Loop
  nodes [Output Animation Pose(Result="Sequence Player (additive) 'Body_Hyde_T2_boost_loop'"() {sequence=&/Body_Hyde_T2_boost_loop})]
 }
 Boost_End {
  path=AnimGraph.AnimGraphNode_StateMachine_2.HydeT2_Boost.AnimStateNode_3.Boost_End
  nodes [Output Animation Pose(Result="Sequence Player (additive) 'Body_Hyde_T2_boost_end'"() {bLoopAnimation=false, sequence=&/Body_Hyde_T2_boost_end})]
 }
 Transition {
  path=AnimGraph.AnimGraphNode_StateMachine_2.HydeT2_Boost.AnimStateTransitionNode_0.Transition
  nodes [Result(bCanEnterTransition=Get bCanEnterTransition_IdleToBoostStart() -> (bCanEnterTransition_IdleToBoostStart:bool))]
 }
 Transition {
  path=AnimGraph.AnimGraphNode_StateMachine_2.HydeT2_Boost.AnimStateTransitionNode_1.Transition
  nodes [Result(bCanEnterTransition=Get bCanEnterTransition_BoostStartToBoostLoop() -> (bCanEnterTransition_BoostStartToBoostLoop:bool))]
 }
 Transition {
  path=AnimGraph.AnimGraphNode_StateMachine_2.HydeT2_Boost.AnimStateTransitionNode_2.Transition
  nodes [Result(bCanEnterTransition=Get bCanEnterTransition_BoostLoopToBoostEnd() -> (bCanEnterTransition_BoostLoopToBoostEnd:bool))]
 }
 Transition {
  path=AnimGraph.AnimGraphNode_StateMachine_2.HydeT2_Boost.AnimStateTransitionNode_3.Transition
  nodes [Result(bCanEnterTransition=Get bCanEnterTransition_BoostEndToIdle() -> (bCanEnterTransition_BoostEndToIdle:bool))]
 }
 EventGraph {nodes [
   TryGetPawnOwner()
   Event BlueprintUpdateAnimation()
  ]}
]
variables []
